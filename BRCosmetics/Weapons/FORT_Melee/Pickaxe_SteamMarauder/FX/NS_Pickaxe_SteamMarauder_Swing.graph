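# Niagara swing trail effect for the SteamMarauder pickaxe, built as a single short-duration MainTrail ribbon emitter.
# FN AnimTrail modules spawn, initialize and update particles from two bone or socket indices on the system's skeletal mesh, feeding two ribbon renderers.
# Short-lived particles narrow and fade over their lifetime through curves.
# Time-of-day inputs tint both the translucent pale base color and a separate dark purple-magenta streak color.
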NiagaraSystem /BRCosmetics/Weapons/FORT_Melee/Pickaxe_SteamMarauder/FX/NS_Pickaxe_SteamMarauder_Swing
dependencies {
 usedDataInterfaces [&NiagaraDataInterfaceCurve, &NiagaraDataInterfaceSkeletalMesh]
 usedDynamicInputs []
 usedModules [
  &self:SystemSpawnScript.SystemScriptSource.SystemScriptGraph.NiagaraNodeAssignment_0.SetVariables_#1
  &SystemState
 ]
}
emitters {MainTrail {
  emitterSpawnScript [
   /SetVariables(Emitter.SpherePos:Vector4f=<0 0.75 0 0>) {enabled=false, isSetParameters=true}
   /SetVariables(Emitter.BottomFade=-0.8) {isSetParameters=true}
  ]
  emitterUpdateScript [
   &EmitterState(Life Cycle Mode=::NewEnumerator1, Loop Behavior=::NewEnumerator1, Loop Duration=0.3, UseLoopDelay=false)
   &FN_SpawnAnimTrail()
  ]
  particleSpawnScript [
   &FN_InitializeAnimTrailBoneIndecies(First Bone or Socket Index=1, Second Bone or Socket Index=2)
   &FN_InitializeAnimTrail(
    AnimTrailWidth=1.5
    Color=&FNTimeOfDayColor(Night Color=<rgba 0.4225 0.6232 1 0.4>, Morning Color=<rgba 1 1 1 0.4>, Day Color=<rgba 1 1 1 0.4>, Evening Color=<rgba 1 0.9295 0.6746 0.4>)
    Lifetime=0.15
    SkeletalMesh=$System.Skeletal Mesh
   )
  ]
  particleUpdateScript [
   &FN_UpdateAnimTrail(SkeletalMesh=$System.Skeletal Mesh) {module=FN_UpdateAnimTrail001}
   &ScaleRibbonWidth(Ribbon Width Scale=&FloatFromCurve(FloatCurve=DataInterface() {propertyValues {Curve {keys [
         {arriveTangent=-0.0405, interpMode=RCIM_Cubic, leaveTangent=-0.0405, tangentMode=RCTM_User, time=0, value=1}
         {arriveTangent=-2.3676, interpMode=RCIM_Cubic, leaveTangent=-2.3676, tangentMode=RCTM_User, time=1, value=0}
        ]}}}))
   /SetVariables(
    Particles.Streak Color:LinearColor=&FNTimeOfDayColor(
     Night Color=<rgba 0.0563 0.0078 0.0573 2>
     Morning Color=<rgba 0.2188 0 0.1014 1.5>
     Day Color=<rgba 0.2188 0 0.1014 1.5>
     Evening Color=<rgba 0.2188 0 0.1014 1.5>
    )
    Particles.StreakWidth=5
   ) {isSetParameters=true}
   &ScaleColor(
    Scale Alpha=&FloatFromCurve(
     FloatCurve=DataInterface() {propertyValues {Curve {keys [{arriveTangent=-4.2427, interpMode=RCIM_Cubic, leaveTangent=-4.2427, tangentMode=RCTM_User, time=0.3, value=1}, <at 0.9 0>]}}}
     Scale Curve=0.25
    )
    ScaleA=true
    ScaleRGB=true
   ) {module=ScaleColor001}
   &ScaleColor(
    Scale Alpha=&FloatFromCurve(FloatCurve=DataInterface() {propertyValues {Curve {keys [{arriveTangent=-5.244, interpMode=RCIM_Cubic, leaveTangent=-5.244, tangentMode=RCTM_User, time=0.75, value=1}, <at 1 0>]}}})
    ScaleA=true
    ScaleRGB=true
   ) {enabled=false}
  ]
  rendererClasses [&NiagaraRibbonRendererProperties]
  renderers [
   {rendererClass=&NiagaraRibbonRendererProperties}
   {rendererClass=&NiagaraRibbonRendererProperties, rendererIndex=1}
  ]
 }}
summary {
 systemName=NS_Pickaxe_SteamMarauder_Swing
 emitters [{emitterName=MainTrail, rendererClasses [&NiagaraRibbonRendererProperties]}]
 userVariables [
  User.DebugVelocity {type=NiagaraFloat, defaultValue=0, description=""}
  User.RibbonIDOffset {type=NiagaraInt32}
  User.SwingSectionIndex {type=NiagaraInt32}
 ]
}
