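# Physics asset for the Steam Marauder pickaxe, defining ragdoll-style bodies and joints for a three-link dangling chain on the weapon.
# Small capsules sit on the Base bone and each dyn_chain bone, with constraints linking them in series from Base out to the third link.
# Every joint nearly locks twist while allowing limited swing, with an asymmetric swing range where the chain meets Base and equal ranges between links.
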PhysicsAsset /BRCosmetics/Weapons/FORT_Melee/Pickaxe_SteamMarauder/Meshes/SteamMarauder_Axe_Physics
bodies {
 Base [{shapeType=Capsule, length=1, radius=1}]
 dyn_chain_01 [{shapeType=Capsule, length=1, radius=1, center=<0.5398 0 0>}]
 dyn_chain_02 [{shapeType=Capsule, length=1, radius=1, center=<0.5398 0 0>}]
 dyn_chain_03 [{shapeType=Capsule, length=1, radius=1, center=<0.5398 0 0>}]
}
constraints [
 {bone1Name=dyn_chain_02, bone2Name=dyn_chain_01, swing1LimitDegrees=45, swing2LimitDegrees=45, twistLimitDegrees=2}
 {bone1Name=dyn_chain_03, bone2Name=dyn_chain_02, swing1LimitDegrees=45, swing2LimitDegrees=45, twistLimitDegrees=2}
 {bone1Name=dyn_chain_01, bone2Name=Base, swing1LimitDegrees=22, swing2LimitDegrees=75, twistLimitDegrees=2}
]
